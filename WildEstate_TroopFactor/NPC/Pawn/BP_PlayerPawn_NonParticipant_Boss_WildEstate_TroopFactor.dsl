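; TroopFactor boss NPC pawn that starts hidden and coordinates its spawn with a TV scare manager actor.
; With spawn animation enabled, the server suspends bot AI at BeginPlay and starts a failsafe timer that applies the spawn GameplayEffect if still hidden.
; IgnoreActorCollision stores the TV, then applies that effect or stims and destroys the TV when animation is off.
; RepNotifies drive visibility and TV collision ignores.
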(event Custom|SetHidePawn
  (params (HidePawn :type Boolean))
  (Networking|SwitchHasAuthority
    (:Authority
      (Variables|Default|SetHidePawn HidePawn))))

(event EventBeginPlay
  (|CallParentFunction)
  (CallFunction|ForceHideHealthBar 0.0)
  (if (ScalableFloat|GetValueasBool (Variables|Default|GetSpawnAnimationEnabled))
    (Networking|SwitchHasAuthority
      (:Authority
        (Variables|Default|SetUseAlertness false)
        (Variables|Default|SetUseAlertSound false)
        (Utilities|FlowControl|DelayUntilNextTick)
        (bind _asfort_athena_aibot_controller (Utilities|Casting|CastToFortAthenaAIBotController (AI|GetAIController self))
          (:then
            (bind _output_get (Variables|Default|SetBotController _asfort_athena_aibot_controller))
            (AI|PushBehaviorControl _output_get -2097152)
            (Class|FortAthenaAIBotController|StopFire _output_get)
            (AI|ClearFocus _output_get)
            (Class|FortAthenaAIBotController|PauseTargetFocusing _output_get)
            ;[1 1 1] Since we have pawn hidden by default, I want to check we unhide it as expected and if not apply GE that will unhide him and reset behaviors if anything goes wrong within it.
            (bind _returnvalue (Utilities|Time|SetTimerbyEvent (EventDispatchers|CreateEvent :!event Custom|CheckUnhidden) (ScalableFloat|GetValueatLevel (Variables|Default|GetCheckUnhiddenTimer))))
            (Variables|Default|SetFailsafeTimer _returnvalue))
          (:CastFailed
            _))))
    (else
      _))
  (CallFunction|SetHidePawn false))

(event Collision|EventActorBeginOverlap
  (params (OtherActor :type Actor%20Object%20Reference))
  (|CallParentFunction OtherActor) ; This node is disabled and will not be called.\nDrag off pins to build functionality.
)

(event EventTick
  (params (DeltaSeconds :type Float%20%28single-precision%29))
  (|CallParentFunction DeltaSeconds) ; This node is disabled and will not be called.\nDrag off pins to build functionality.
)

(event Custom|CheckUnhidden
  ;[1 1 1] Since we have pawn hidden by default, I want to check we unhide it as expected and if not apply GE that will unhide him and reset behaviors if anything goes wrong within it.
  (if (Variables|Rendering|GetbHidden)
    (GameplayEffects|ApplyGameplayEffectToSelf :self (Ability|GetFortAbilitySystemComponent self) :GameplayEffectClass "/WildEstate_TroopFactor/Gameplay/Abilites/GE_OnSpawn_WildEstate_TroopFactor.GE_OnSpawn_WildEstate_TroopFactor_C" :Level 0.0)
    (Fortnite|FortShippingLog self "FAILSAFE: TroopFactor pawn was not unhidden as expected and is being unhidden by failsafe catch" true)))

(event Custom|IgnoreActorCollision
  (params (IgnoreActor :type Actor%20Object%20Reference))
  (bind _scaretv (Variables|Default|GetScareTV))
  (bind _asbp_tvscare_troop_factor_manager (Utilities|Casting|CastToBP_TVScare_TroopFactor_Manager IgnoreActor))
  (Variables|Default|SetScareTV _asbp_tvscare_troop_factor_manager)
  (if (ScalableFloat|GetValueasBool (Variables|Default|GetSpawnAnimationEnabled))
    (GameplayEffects|ApplyGameplayEffectToSelf (Ability|GetFortAbilitySystemComponent self) "/WildEstate_TroopFactor/Gameplay/Abilites/GE_OnSpawn_WildEstate_TroopFactor.GE_OnSpawn_WildEstate_TroopFactor_C" 0.0 (Class|AthenaFunctionLibrary|MakeSimpleFortEffectContext (Variables|Default|GetScareTV)))
    (else
      (Class|BPTVScareTroopFactorManager|StimTarget _scaretv)
      (Actor|DestroyActor _scaretv))))

(event EventEndPlay
  (params (EndPlayReason :type EEndPlayReason%20Enum :default "Destroyed"))
  (Utilities|Time|ClearandInvalidateTimerbyHandle (Variables|Default|GetFailsafeTimer)))

(fn OnRep_HidePawn
  (Rendering|SetActorHiddenInGame (Variables|Default|GetHidePawn)))

(fn OnRep_ScareTV
  (bind _scaretv (Variables|Default|GetScareTV))
  (Collision|AddIgnoredActor _scaretv 1.2e+02)
  (Collision|IgnoreActorwhenMoving (Variables|Character|GetCapsuleComponent) _scaretv true)
  (Collision|IgnoreActorwhenMoving (Variables|Character|GetMesh) _scaretv true))

(fn ConstructionScript
  (|CallParentFunction))
